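# Physics asset for the choker on the F_MED_Lollipop female character body, giving its dynamic front collar bones collision shapes for simulated secondary motion.
# The collar top bone carries a compound shape of two capsules and three boxes, two of them angled, while its master bone has a tiny capsule.
# Capsules on neck_01 and spine_05 approximate neck and upper torso.
# Constraints chain collar top, master, neck and spine with free angular motion.
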PhysicsAsset /BRCosmetics/Characters/Player/Female/Medium/Bodies/F_MED_Lollipop/Meshes/F_MED_Lollipop_Dyn_choker
bodies {
 dyn_collar_fr_MSTR [{shapeType=Capsule, length=0.5, radius=0.5}]
 dyn_collar_fr_top [
  {shapeType=Capsule, length=5.2954, radius=1.2665, center=<-1.1001 1.4095 0>}
  {shapeType=Capsule, length=5.2954, radius=1.2665, center=<0.7654 1.4095 0>}
  {shapeType=Box, extentX=1.8523, extentY=0.9731, extentZ=7.5, center=<-0.1589 0.648 0>}
  {shapeType=Box, extentX=3.0649, extentY=0.9731, extentZ=7.5, center=<-0.1928 -0.6383 2.9095>, rotation=<rot 0 0 -44.9998>}
  {shapeType=Box, extentX=3.0649, extentY=0.9731, extentZ=7.5, center=<-0.2046 -1.1469 -3.5689>, rotation=<rot 0 0 45>}
 ]
 neck_01 [{shapeType=Capsule, length=8.6471, radius=3.7188, center=<2.8007 1.0012 -0.0001>, rotation=<rot 90 0 0>}]
 spine_05 [{shapeType=Capsule, length=12.8031, radius=8.6478, center=<2.2488 -1.9364 -0.6874>, rotation=<rot -84.9994 -0.0001 -20.0002>}]
}
constraints [
 {
  bone1Name=dyn_collar_fr_MSTR
  bone2Name=neck_01
  swing1LimitDegrees=45
  swing1Motion=Free
  swing2LimitDegrees=45
  swing2Motion=Free
  twistLimitDegrees=45
  twistMotion=Free
 }
 {
  bone1Name=dyn_collar_fr_top
  bone2Name=dyn_collar_fr_MSTR
  swing1LimitDegrees=45
  swing1Motion=Free
  swing2LimitDegrees=45
  swing2Motion=Free
  twistLimitDegrees=45
  twistMotion=Free
 }
 {
  bone1Name=neck_01
  bone2Name=spine_05
  swing1LimitDegrees=45
  swing1Motion=Free
  swing2LimitDegrees=45
  swing2Motion=Free
  twistLimitDegrees=45
  twistMotion=Free
 }
]
