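# Physics asset for the left-leg dynamic bones of the female medium Lollipop body cosmetic, simulating shoelace motion.
# A small master capsule anchors two chains, inner and outer, each made of two capsule bodies.
# Each chain links segment to segment and back to the master through constraints with tight swing and twist limits, keeping lace movement subtle.
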PhysicsAsset /BRCosmetics/Characters/Player/Female/Medium/Bodies/F_MED_Lollipop/Meshes/F_MED_Lollipop_Dyn_leg_l
bodies {
 dyn_shoelace_in_l_01 [{shapeType=Capsule, length=1.5, radius=0.5, center=<1 0 0>, rotation=<rot 90 0 0>}]
 dyn_shoelace_in_l_02 [{shapeType=Capsule, length=1.5, radius=0.5, center=<1 0 0>, rotation=<rot 90 0 0>}]
 dyn_shoelace_l_MSTR [{shapeType=Capsule, length=0.2, radius=0.2}]
 dyn_shoelace_out_l_01 [{shapeType=Capsule, length=1.5, radius=0.5, center=<1 0 0>, rotation=<rot 90 0 0>}]
 dyn_shoelace_out_l_02 [{shapeType=Capsule, length=1.5, radius=0.5, center=<1 0 0>, rotation=<rot 90 0 0>}]
}
constraints [
 {bone1Name=dyn_shoelace_out_l_01, bone2Name=dyn_shoelace_l_MSTR, swing1LimitDegrees=15, swing2LimitDegrees=15, twistLimitDegrees=5}
 {bone1Name=dyn_shoelace_out_l_02, bone2Name=dyn_shoelace_out_l_01, swing1LimitDegrees=15, swing2LimitDegrees=15, twistLimitDegrees=5}
 {bone1Name=dyn_shoelace_in_l_01, bone2Name=dyn_shoelace_l_MSTR, swing1LimitDegrees=15, swing2LimitDegrees=15, twistLimitDegrees=5}
 {bone1Name=dyn_shoelace_in_l_02, bone2Name=dyn_shoelace_in_l_01, swing1LimitDegrees=15, swing2LimitDegrees=15, twistLimitDegrees=5}
]
