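# Physics asset for the torso dynamic bones of the F_MED_Lollipop player body, covering a knife holster and its zipper.
# Three capsule bodies form a chain: a tiny master anchor, the holster, and the zipper pull.
# Angular-limited constraints link holster to master and zipper to holster, with tight twist and a wider first swing on the zipper.
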PhysicsAsset /BRCosmetics/Characters/Player/Female/Medium/Bodies/F_MED_Lollipop/Meshes/F_MED_Lollipop_Dyn_torso
bodies {
 dyn_kinfe_zipper [{shapeType=Capsule, length=1.5, radius=0.5, center=<1 0 0>, rotation=<rot 90 0 0>}]
 dyn_knife_holster [{shapeType=Capsule, length=4.4062, radius=2.1, center=<1 0 0>, rotation=<rot 90 0 0>}]
 dyn_knife_holster_MSTR [{shapeType=Capsule, length=0.2, radius=0.2}]
}
constraints [
 {bone1Name=dyn_knife_holster, bone2Name=dyn_knife_holster_MSTR, swing1LimitDegrees=30, swing2LimitDegrees=25, twistLimitDegrees=5}
 {bone1Name=dyn_kinfe_zipper, bone2Name=dyn_knife_holster, swing1LimitDegrees=45, swing2LimitDegrees=25, twistLimitDegrees=5}
]
