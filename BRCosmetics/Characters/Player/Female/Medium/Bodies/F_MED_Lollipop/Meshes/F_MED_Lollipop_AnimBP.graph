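# Post-process-style AnimBlueprint for the F_MED_Lollipop female player cosmetic body, layering secondary motion onto the pose copied from its attached parent mesh.
# Weighted multi-bone constraints drive collar and knife-holster bones, then additive pelvis bone offsets blend by thigh rotation alphas computed each update.
# Four chained RigidBody nodes simulate torso, both legs and choker in base-bone space with dedicated physics assets.
# Head-rotation alpha setters are unconnected and unused.
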AnimBlueprint /BRCosmetics/Characters/Player/Female/Medium/Bodies/F_MED_Lollipop/Meshes/F_MED_Lollipop_AnimBP
target_skeleton=&/F_MED_Lollipop_Skeleton
graphs [
 AnimGraph {nodes [
   n1 = "Constraint - dyn_collar_fr_MSTR (PPPPP)"!(
    ComponentPose=Local To Component(LocalPose=Copy Pose From Mesh() {bUseAttachedParent=true})
    Alpha=1
    ConstraintWeights_0=0.217
    ConstraintWeights_1:float=0.0858
    ConstraintWeights_2:float=0.1358
    ConstraintWeights_3:float=0.3082
    ConstraintWeights_4:float=0.2532
   ) -> (Pose) {
    alphaBoolBlend {blendOption=Linear}
    boneToModify {boneName=dyn_collar_fr_MSTR}
    constraintSetup [
     {targetBone {boneName=C_jaw}}
     {targetBone {boneName=head}}
     {targetBone {boneName=neck_01}}
     {targetBone {boneName=spine_05}}
     {targetBone {boneName=neck_02}}
    ]
    constraintWeights [1,1,1,1,1]
   }
   n2 = "Constraint - dyn_knife_holster_MSTR (PPPPPPPPP)"!(
    ComponentPose=@n1
    Alpha=1
    ConstraintWeights_0=0.103
    ConstraintWeights_1:float=0
    ConstraintWeights_2:float=0.0002
    ConstraintWeights_3:float=0.0013
    ConstraintWeights_4:float=0.0125
    ConstraintWeights_5:float=0.6348
    ConstraintWeights_6:float=0.157
    ConstraintWeights_7:float=0.0912
    ConstraintWeights_8:float=0
   ) -> (Pose) {
    alphaBoolBlend {blendOption=Linear}
    boneToModify {boneName=dyn_knife_holster_MSTR}
    constraintSetup [
     {targetBone {boneName=clavicle_l}}
     {targetBone {boneName=upperarm_twist_01_r}}
     {targetBone {boneName=spine_01}}
     {targetBone {boneName=spine_02}}
     {targetBone {boneName=spine_03}}
     {targetBone {boneName=spine_04}}
     {targetBone {boneName=spine_05}}
     {targetBone {boneName=upperarm_twist_01_l}}
     {targetBone {boneName=clavicle_r}}
    ]
    constraintWeights [1,1,1,1,1,1,1,1,1]
   }
   n3 = Get thigh_rot_l()
   n4 = "Transform (Modify) Bone - Bone: pelvis_l"(ComponentPose=@n2, Alpha=@n3, Translation="-6.240,-1.520,1.245", Rotation="0,0,0", Scale="1,1,1") {
    scaleMode=BMM_Additive
    translationMode=BMM_Additive
    translationSpace=BCS_ParentBoneSpace
    alphaBoolBlend {blendOption=Linear}
    boneToModify {boneName=pelvis_l}
    translation=<-6.24 -1.52 1.245>
   }
   n5 = "Transform (Modify) Bone - Bone: pelvis_mid_l"(ComponentPose=@n4, Alpha=@n3, Translation="-1.450,-1.138,0.518", Rotation="0,0,0", Scale="1,1,1") {
    scaleMode=BMM_Additive
    translationMode=BMM_Additive
    translationSpace=BCS_ParentBoneSpace
    alphaBoolBlend {blendOption=Linear}
    boneToModify {boneName=pelvis_mid_l}
    translation=<-1.45 -1.138 0.518>
   }
   n6 = "Transform (Modify) Bone - Bone: pelvis_side_l"(ComponentPose=@n5, Alpha=@n3, Translation="0.505,-0.560,-1.440", Rotation="0,0,0", Scale="1,1,1") {
    scaleMode=BMM_Additive
    translationMode=BMM_Additive
    translationSpace=BCS_ParentBoneSpace
    alphaBoolBlend {blendOption=Linear}
    boneToModify {boneName=pelvis_side_l}
    translation=<0.505 -0.56 -1.44>
   }
   n7 = Get thigh_rot_r()
   n8 = "Transform (Modify) Bone - Bone: pelvis_r"(ComponentPose=@n6, Alpha=@n7, Translation="-1.106,-0.155,0.183", Rotation="0,0,0", Scale="1,1,1") {
    scaleMode=BMM_Additive
    translationMode=BMM_Additive
    translationSpace=BCS_ParentBoneSpace
    alphaBoolBlend {blendOption=Linear}
    boneToModify {boneName=pelvis_r}
    translation=<-1.106 -0.155 0.183>
   }
   n9 = "Transform (Modify) Bone - Bone: pelvis_mid_r"(ComponentPose=@n8, Alpha=@n7, Translation="-0.848,0.002,0.106", Rotation="0,0,0", Scale="1,1,1") {
    scaleMode=BMM_Additive
    translationMode=BMM_Additive
    translationSpace=BCS_ParentBoneSpace
    alphaBoolBlend {blendOption=Linear}
    boneToModify {boneName=pelvis_mid_r}
    translation=<-0.848 0.002 0.106>
   }
   n10 = "Transform (Modify) Bone - Bone: pelvis_side_r"(ComponentPose=@n9, Alpha=@n7, Translation="-0.144,-0.340,-1.135", Rotation="0,0,0", Scale="1,1,1") {
    scaleMode=BMM_Additive
    translationMode=BMM_Additive
    translationSpace=BCS_ParentBoneSpace
    alphaBoolBlend {blendOption=Linear}
    boneToModify {boneName=pelvis_side_r}
    translation=<-0.144 -0.34 -1.135>
   }
   n11 = RigidBody(ComponentPose=@n10) {
    bClampLinearTranslationLimitToRefPose=true
    bOverrideWorldGravity=true
    overridePhysicsAsset=&/F_MED_Lollipop_Dyn_torso
    simulationSpace=BaseBoneSpace
    alphaBoolBlend {blendOption=Linear}
    baseBoneRef {boneName=spine_04}
    componentAppliedLinearAccClamp=<3000 3000 3000>
    componentLinearAccScale=<0.25 0.15 0.15>
    componentLinearVelScale=<0.25 0.15 0.15>
    overrideWorldGravity {z=-980}
   }
   n12 = RigidBody(ComponentPose=@n11) {
    bClampLinearTranslationLimitToRefPose=true
    bOverrideWorldGravity=true
    overridePhysicsAsset=&/F_MED_Lollipop_Dyn_leg_l
    simulationSpace=BaseBoneSpace
    alphaBoolBlend {blendOption=Linear}
    baseBoneRef {boneName=thigh_l}
    componentAppliedLinearAccClamp=<3000 3000 3000>
    componentLinearAccScale=<0.4 0.3 0.3>
    componentLinearVelScale=<0.4 0.3 0.3>
    overrideWorldGravity {z=-980}
   }
   n13 = RigidBody(ComponentPose=@n12) {
    bClampLinearTranslationLimitToRefPose=true
    bOverrideWorldGravity=true
    overridePhysicsAsset=&/F_MED_Lollipop_Dyn_leg_r
    simulationSpace=BaseBoneSpace
    alphaBoolBlend {blendOption=Linear}
    baseBoneRef {boneName=thigh_r}
    componentAppliedLinearAccClamp=<3000 3000 3000>
    componentLinearAccScale=<0.4 0.3 0.3>
    componentLinearVelScale=<0.4 0.3 0.3>
    overrideWorldGravity {z=-980}
   }
   n14 = RigidBody(ComponentPose=@n13) {
    bClampLinearTranslationLimitToRefPose=true
    bOverrideWorldGravity=true
    overridePhysicsAsset=&/F_MED_Lollipop_Dyn_choker
    simulationSpace=BaseBoneSpace
    alphaBoolBlend {blendOption=Linear}
    baseBoneRef {boneName=neck_02}
    componentAppliedLinearAccClamp=<0 0 0>
   }
   Output Pose(Result=Component To Local(ComponentPose=@n14)) {layerGroup=None}
  ]}
 EventGraph {nodes [
   Event BlueprintUpdateAnimation()
   n1 = Get Skel_Mesh()
   > Set thigh_rot_r(thigh_rot_r=GetAlphaFromReferenceTransformDelta(SkeletalMeshComponent=@n1, BoneName=thigh_r, BaseName=pelvis, InputRange=(X=0,Y=-90), TwistAxis="0,0,1"))
   > Set thigh_rot_l(thigh_rot_l=GetAlphaFromReferenceTransformDelta(SkeletalMeshComponent=@n1, BoneName=thigh_l, BaseName=pelvis, InputRange=(X=-15,Y=-110), TwistAxis="0,0,1"))
   Event BlueprintInitializeAnimation()
   > Set Skel_Mesh(Skel_Mesh=GetOwnerSkeletalMeshComponent())
   Set head_rot_up(head_rot_up:double=GetAlphaFromReferenceTransformDelta(SkeletalMeshComponent=@n1, BoneName=head, BaseName=spine_05, InputRange=(X=-5,Y=-43), TwistAxis="0,0,1")) -> (Output_Get:double)
   > Set head_rot_down(head_rot_down:double=GetAlphaFromReferenceTransformDelta(SkeletalMeshComponent=@n1, BoneName=head, BaseName=spine_05, InputRange=(X=-5,Y=20), TwistAxis="0,0,1")) -> (Output_Get:double)
  ]}
]
variables [
 head_rot_down {type=number}
 head_rot_up {type=number}
 Skel_Mesh {type=SkeletalMeshComponent}
 thigh_rot_l {type=number}
 thigh_rot_r {type=number}
]
