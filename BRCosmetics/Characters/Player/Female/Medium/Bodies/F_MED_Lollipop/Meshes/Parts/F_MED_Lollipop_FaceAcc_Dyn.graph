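# Physics asset for the F_MED_Lollipop female character's face accessory mesh, set up for dynamic left and right ponytails.
# Each ponytail is a capsule chain from a holder on the head through six bones, with swing limits widening down the chain and tight twist.
# Head, neck, upper spine, clavicle and upper arm bodies, joined by free constraints, give the hair collision shapes.
# The sides differ slightly in capsule sizes and holder limits.
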PhysicsAsset /BRCosmetics/Characters/Player/Female/Medium/Bodies/F_MED_Lollipop/Meshes/Parts/F_MED_Lollipop_FaceAcc_Dyn
bodies {
 clavicle_l [{shapeType=Capsule, length=7.4062, radius=6.5, center=<5.5757 -1.0268 -1.5864>, rotation=<rot 84.9996 0 0>}]
 clavicle_r [{shapeType=Capsule, length=7.4062, radius=6.5, center=<-5.5757 1.0268 1.5864>, rotation=<rot 84.9996 0 0>}]
 dyn_ponytail_l_01 [{shapeType=Capsule, length=2, radius=1, center=<2 0 0>, rotation=<rot 90 0 0>}]
 dyn_ponytail_l_02 [{shapeType=Capsule, length=3, radius=2, center=<3 0 0>, rotation=<rot 90 0 0>}]
 dyn_ponytail_l_03 [{shapeType=Capsule, length=0.25, radius=5, center=<3 0 0>, rotation=<rot 90 0 0>}]
 dyn_ponytail_l_04 [{shapeType=Capsule, length=0.25, radius=5, center=<3 0 0>, rotation=<rot 90 0 0>}]
 dyn_ponytail_l_05 [{shapeType=Capsule, length=1.9809, radius=3.5, center=<3 0 0>, rotation=<rot 90 0 0>}]
 dyn_ponytail_l_06 [{shapeType=Capsule, length=5.1875, radius=1.7188, center=<3.9419 0 0>, rotation=<rot 90 0 0>}]
 dyn_ponytail_l_Holder [{shapeType=Capsule, length=2, radius=1, center=<2 0 0>, rotation=<rot 90 0 0>}]
 dyn_ponytail_r_01 [{shapeType=Capsule, length=2, radius=1, center=<3 0 0>, rotation=<rot 90 0 0>}]
 dyn_ponytail_r_02 [{shapeType=Capsule, length=3, radius=2, center=<3 0 0>, rotation=<rot 90 0 0>}]
 dyn_ponytail_r_03 [{shapeType=Capsule, length=1, radius=4, center=<3 0 0>, rotation=<rot 90 0 0>}]
 dyn_ponytail_r_04 [{shapeType=Capsule, length=0.25, radius=4.75, center=<3 0 0>, rotation=<rot 90 0 0>}]
 dyn_ponytail_r_05 [{shapeType=Capsule, length=0.9809, radius=2.75, center=<3 0 0>, rotation=<rot 90 0 0>}]
 dyn_ponytail_r_06 [{shapeType=Capsule, length=7.3438, radius=1, center=<3.9736 0 0>, rotation=<rot 90 0 0>}]
 dyn_ponytail_r_Holder [{shapeType=Capsule, length=2, radius=1, center=<2 0 0>, rotation=<rot 90 0 0>}]
 head [
  {shapeType=Sphere, radius=7.5534, center=<6.1928 0.708 0>}
  {shapeType=Capsule, length=6.7188, radius=6.75, center=<1.1967 2.6705 0>, rotation=<rot 90 0 0>}
 ]
 neck_02 [{shapeType=Capsule, length=4.5, radius=5.5, rotation=<rot 90 0 0>}]
 spine_05 [{shapeType=Capsule, length=8.505, radius=10.005, center=<0.833 0 0>, rotation=<rot 0 -90 0>}]
 upperarm_l [{shapeType=Capsule, length=17.5, radius=4.755, center=<10.702 -0.0001 0>, rotation=<rot 90 0 0>}]
 upperarm_r [{shapeType=Capsule, length=17.5, radius=4.755, center=<-10.702 0.0001 0>, rotation=<rot 90 0 180>}]
}
constraints [
 {
  bone1Name=head
  bone2Name=neck_02
  swing1LimitDegrees=45
  swing1Motion=Free
  swing2LimitDegrees=45
  swing2Motion=Free
  twistLimitDegrees=45
  twistMotion=Free
 }
 {bone1Name=dyn_ponytail_l_Holder, bone2Name=head, swing1LimitDegrees=1, swing2LimitDegrees=1, twistLimitDegrees=1}
 {bone1Name=dyn_ponytail_l_01, bone2Name=dyn_ponytail_l_Holder, swing1LimitDegrees=10, swing2LimitDegrees=10, twistLimitDegrees=5}
 {bone1Name=dyn_ponytail_l_02, bone2Name=dyn_ponytail_l_01, swing1LimitDegrees=15, swing2LimitDegrees=15, twistLimitDegrees=5}
 {bone1Name=dyn_ponytail_l_03, bone2Name=dyn_ponytail_l_02, swing1LimitDegrees=35, swing2LimitDegrees=35, twistLimitDegrees=5}
 {bone1Name=dyn_ponytail_l_04, bone2Name=dyn_ponytail_l_03, swing1LimitDegrees=35, swing2LimitDegrees=35, twistLimitDegrees=5}
 {bone1Name=dyn_ponytail_l_05, bone2Name=dyn_ponytail_l_04, swing1LimitDegrees=35, swing2LimitDegrees=35, twistLimitDegrees=5}
 {bone1Name=dyn_ponytail_l_06, bone2Name=dyn_ponytail_l_05, swing1LimitDegrees=35, swing2LimitDegrees=35, twistLimitDegrees=5}
 {bone1Name=dyn_ponytail_r_Holder, bone2Name=head, swing1LimitDegrees=5, swing2LimitDegrees=5, twistLimitDegrees=5}
 {bone1Name=dyn_ponytail_r_01, bone2Name=dyn_ponytail_r_Holder, swing1LimitDegrees=10, swing2LimitDegrees=10, twistLimitDegrees=5}
 {bone1Name=dyn_ponytail_r_02, bone2Name=dyn_ponytail_r_01, swing1LimitDegrees=15, swing2LimitDegrees=15, twistLimitDegrees=5}
 {bone1Name=dyn_ponytail_r_03, bone2Name=dyn_ponytail_r_02, swing1LimitDegrees=35, swing2LimitDegrees=35, twistLimitDegrees=5}
 {bone1Name=dyn_ponytail_r_04, bone2Name=dyn_ponytail_r_03, swing1LimitDegrees=35, swing2LimitDegrees=35, twistLimitDegrees=5}
 {bone1Name=dyn_ponytail_r_05, bone2Name=dyn_ponytail_r_04, swing1LimitDegrees=35, swing2LimitDegrees=35, twistLimitDegrees=5}
 {bone1Name=dyn_ponytail_r_06, bone2Name=dyn_ponytail_r_05, swing1LimitDegrees=35, swing2LimitDegrees=35, twistLimitDegrees=5}
 {
  bone1Name=clavicle_l
  bone2Name=spine_05
  swing1LimitDegrees=45
  swing1Motion=Free
  swing2LimitDegrees=45
  swing2Motion=Free
  twistLimitDegrees=45
  twistMotion=Free
 }
 {
  bone1Name=clavicle_r
  bone2Name=spine_05
  swing1LimitDegrees=45
  swing1Motion=Free
  swing2LimitDegrees=45
  swing2Motion=Free
  twistLimitDegrees=45
  twistMotion=Free
 }
 {
  bone1Name=neck_02
  bone2Name=spine_05
  swing1LimitDegrees=45
  swing1Motion=Free
  swing2LimitDegrees=45
  swing2Motion=Free
  twistLimitDegrees=45
  twistMotion=Free
 }
 {
  bone1Name=upperarm_l
  bone2Name=clavicle_l
  swing1LimitDegrees=45
  swing1Motion=Free
  swing2LimitDegrees=45
  swing2Motion=Free
  twistLimitDegrees=45
  twistMotion=Free
 }
 {
  bone1Name=upperarm_r
  bone2Name=clavicle_r
  swing1LimitDegrees=45
  swing1Motion=Free
  swing2LimitDegrees=45
  swing2Motion=Free
  twistLimitDegrees=45
  twistMotion=Free
 }
]
