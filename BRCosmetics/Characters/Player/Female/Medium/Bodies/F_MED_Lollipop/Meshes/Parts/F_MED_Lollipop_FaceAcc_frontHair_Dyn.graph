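# Physics asset for the dynamic front hair bangs of the female medium Lollipop character's face accessory.
# Two six-bone chains, left and right, hang from the head, with one capsule per bone.
# Each runs in series from 01 to 04, then forks into 05 and 06, under uniform tight swing and twist limits.
# The head carries a sphere and a capsule as the body both chains anchor to.
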PhysicsAsset /BRCosmetics/Characters/Player/Female/Medium/Bodies/F_MED_Lollipop/Meshes/Parts/F_MED_Lollipop_FaceAcc_frontHair_Dyn
bodies {
 dyn_bangs_long_l_01 [{shapeType=Capsule, length=2, radius=1, center=<2 0 0>, rotation=<rot 90 0 0>}]
 dyn_bangs_long_l_02 [{shapeType=Capsule, length=2.4375, radius=1.125, center=<2.2257 0 0>, rotation=<rot 90 0 0>}]
 dyn_bangs_long_l_03 [{shapeType=Capsule, length=3.1562, radius=1.375, center=<2.7665 0 0>, rotation=<rot 90 0 0>}]
 dyn_bangs_long_l_04 [{shapeType=Capsule, length=2, radius=1, center=<2 0 0>, rotation=<rot 90 0 0>}]
 dyn_bangs_long_l_05 [{shapeType=Capsule, length=4.8125, radius=1, center=<3.1184 0 0>, rotation=<rot 90 0 0>}]
 dyn_bangs_long_l_06 [{shapeType=Capsule, length=2, radius=1, center=<2.1788 -1.2973 0>, rotation=<rot 90 0 0>}]
 dyn_bangs_rong_r_01 [{shapeType=Capsule, length=2, radius=1, center=<2 0 0>, rotation=<rot 90 0 0>}]
 dyn_bangs_rong_r_02 [{shapeType=Capsule, length=2.4688, radius=1.0938, center=<2.1852 0 0>, rotation=<rot 90 0 0>}]
 dyn_bangs_rong_r_03 [{shapeType=Capsule, length=3.5625, radius=1.1562, center=<2.7574 0 0>, rotation=<rot 90 0 0>}]
 dyn_bangs_rong_r_04 [{shapeType=Capsule, length=2, radius=1, center=<2 0 0>, rotation=<rot 90 0 0>}]
 dyn_bangs_rong_r_05 [{shapeType=Capsule, length=3.8438, radius=1, center=<2.3824 0 0>, rotation=<rot 90 0 0>}]
 dyn_bangs_rong_r_06 [{shapeType=Capsule, length=2, radius=1, center=<2 0 0>, rotation=<rot 90 0 0>}]
 head [
  {shapeType=Sphere, radius=7.5534, center=<6.1928 0.708 0>}
  {shapeType=Capsule, length=6.7188, radius=6.75, center=<1.1967 2.6705 0>, rotation=<rot 90 0 0>}
 ]
}
constraints [
 {bone1Name=dyn_bangs_long_l_01, bone2Name=head, swing1LimitDegrees=10, swing2LimitDegrees=10, twistLimitDegrees=5}
 {bone1Name=dyn_bangs_long_l_02, bone2Name=dyn_bangs_long_l_01, swing1LimitDegrees=10, swing2LimitDegrees=10, twistLimitDegrees=5}
 {bone1Name=dyn_bangs_long_l_03, bone2Name=dyn_bangs_long_l_02, swing1LimitDegrees=10, swing2LimitDegrees=10, twistLimitDegrees=5}
 {bone1Name=dyn_bangs_long_l_04, bone2Name=dyn_bangs_long_l_03, swing1LimitDegrees=10, swing2LimitDegrees=10, twistLimitDegrees=5}
 {bone1Name=dyn_bangs_long_l_05, bone2Name=dyn_bangs_long_l_04, swing1LimitDegrees=10, swing2LimitDegrees=10, twistLimitDegrees=5}
 {bone1Name=dyn_bangs_long_l_06, bone2Name=dyn_bangs_long_l_04, swing1LimitDegrees=10, swing2LimitDegrees=10, twistLimitDegrees=5}
 {bone1Name=dyn_bangs_rong_r_01, bone2Name=head, swing1LimitDegrees=10, swing2LimitDegrees=10, twistLimitDegrees=5}
 {bone1Name=dyn_bangs_rong_r_02, bone2Name=dyn_bangs_rong_r_01, swing1LimitDegrees=10, swing2LimitDegrees=10, twistLimitDegrees=5}
 {bone1Name=dyn_bangs_rong_r_03, bone2Name=dyn_bangs_rong_r_02, swing1LimitDegrees=10, swing2LimitDegrees=10, twistLimitDegrees=5}
 {bone1Name=dyn_bangs_rong_r_04, bone2Name=dyn_bangs_rong_r_03, swing1LimitDegrees=10, swing2LimitDegrees=10, twistLimitDegrees=5}
 {bone1Name=dyn_bangs_rong_r_05, bone2Name=dyn_bangs_rong_r_04, swing1LimitDegrees=10, swing2LimitDegrees=10, twistLimitDegrees=5}
 {bone1Name=dyn_bangs_rong_r_06, bone2Name=dyn_bangs_rong_r_04, swing1LimitDegrees=10, swing2LimitDegrees=10, twistLimitDegrees=5}
]
