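# Animation Blueprint for the Lollipop female character's face accessory part, adding physics motion to the pose copied from its parent mesh.
# A cached pose layers body-mesh arms over the parent pose, then feeds separate gameplay and front-end RigidBody simulations, selected by bIsInFrontEnd.
# A front-hair RigidBody follows, then additive yaw splays long bangs during skydive floating.
# Skydiving state and vertical velocity vary gameplay gravity; frontHairSkydivingAlpha is set but drives no node.
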AnimBlueprint /BRCosmetics/Characters/Player/Female/Medium/Bodies/F_MED_Lollipop/Meshes/Parts/F_MED_Lollipop_FaceAcc_AnimBP
target_skeleton=&/F_MED_Lollipop_FaceAcc_Skeleton
graphs [
 AnimGraph {nodes [
   Copy Pose From Mesh() {bUseAttachedParent=true}
   n1 = RigidBody!(
    ComponentPose=Local To Component(LocalPose=Use cached pose 'prerigid'())
    Alpha
    OverrideWorldGravity:Vector=Get gravity_override() -> (gravity_override:Vector)
    ExternalForce
   ) -> (Pose) {
    bClampLinearTranslationLimitToRefPose=true
    bOverrideWorldGravity=true
    overridePhysicsAsset=&/F_MED_Lollipop_FaceAcc_Dyn
    simulationSpace=BaseBoneSpace
    alphaBoolBlend {blendOption=Linear}
    baseBoneRef {boneName=spine_02}
    componentAppliedLinearAccClamp=<3000 3000 3000>
    componentLinearAccScale=<0.3 0.3 0.4>
    componentLinearVelScale=<0.3 0.3 0.4>
    overrideWorldGravity {z=-980}
   }
   n2 = Component To Local(ComponentPose=@n1)
   n3 = RigidBody(ComponentPose=Local To Component(LocalPose=Use cached pose 'prerigid'()), OverrideWorldGravity="0,0,-980") {
    bClampLinearTranslationLimitToRefPose=true
    bOverrideWorldGravity=true
    overridePhysicsAsset=&/F_MED_Lollipop_FaceAcc_FrontEnd_Dyn
    simulationSpace=BaseBoneSpace
    alphaBoolBlend {blendOption=Linear}
    baseBoneRef {boneName=root}
    componentAppliedLinearAccClamp=<3000 3000 3000>
    componentLinearAccScale=<0.3 0.3 0.4>
    componentLinearVelScale=<0.3 0.3 0.4>
    overrideWorldGravity {z=-980}
   }
   n4 = Component To Local(ComponentPose=@n3)
   n5 = Get bIsInFrontEnd()
   n6 = RigidBody(
    ComponentPose=Local To Component(LocalPose=Blend Poses by bool(BlendPose_0=@n4, BlendPose_1=@n2, bActiveValue=@n5) {blendType=Linear})
    OverrideWorldGravity="0,0,-980"
   ) {
    bClampLinearTranslationLimitToRefPose=true
    bOverrideWorldGravity=true
    overridePhysicsAsset=&/F_MED_Lollipop_FaceAcc_frontHair_Dyn
    simulationSpace=BaseBoneSpace
    alphaBoolBlend {blendOption=Linear}
    baseBoneRef {boneName=spine_03}
    componentAppliedLinearAccClamp=<3000 3000 3000>
    componentLinearAccScale=<0.3 0.2 0.2>
    componentLinearVelScale=<0.3 0.2 0.2>
    overrideWorldGravity {z=-980}
   }
   n7 = Get bIsSkydiveFloating()
   n8 = "Transform (Modify) Bone - Bone: dyn_bangs_long_l_01"!(ComponentPose=@n6, bAlphaBoolEnabled:bool=@n7, Translation="0,0,0", Rotation="0,-30,0", Scale="1,1,1") -> (Pose) {
    alphaInputType=Bool
    rotationMode=BMM_Additive
    rotationSpace=BCS_BoneSpace
    translationMode=BMM_Additive
    translationSpace=BCS_BoneSpace
    alphaBoolBlend {blendInTime=1, blendOption=Linear, blendOutTime=1}
    boneToModify {boneName=dyn_bangs_long_l_01}
    rotation {yaw=-30}
   }
   n9 = "Transform (Modify) Bone - Bone: dyn_bangs_rong_r_01"!(ComponentPose=@n8, bAlphaBoolEnabled:bool=@n7, Translation="0,0,0", Rotation="0,30,0", Scale="1,1,1") -> (Pose) {
    alphaInputType=Bool
    rotationMode=BMM_Additive
    rotationSpace=BCS_BoneSpace
    translationMode=BMM_Additive
    translationSpace=BCS_BoneSpace
    alphaBoolBlend {blendOption=Linear}
    boneToModify {boneName=dyn_bangs_rong_r_01}
    rotation {yaw=30}
   }
   Output Pose(Result=Component To Local(ComponentPose=@n9)) {layerGroup=None}
   n10 = Get Body_Mesh()
   Save cached pose 'prerigid'(Pose=Layered blend per bone(BasePose=Copy Pose From Mesh() {bCopyCurves=true, bUseAttachedParent=true}, BlendPoses_0=Copy Pose From Mesh(SourceMeshComponent=@n10)) {layerSetup [{branchFilters [{boneName=upperarm_l}, {boneName=upperarm_r}]}]}) {cacheName=prerigid}
   Get frontHairSkydivingAlpha() -> (frontHairSkydivingAlpha:double)
  ]}
 EventGraph {nodes [
   Event BlueprintUpdateAnimation()
   > n1 = Sequence()
   > n2 = Branch(Condition=Get bIsSkyDiving())
   > n3 = Branch(Condition=Get bIsParachuteOpened())
   Set gravity_override(execute=[@n2.else, @n3.then], gravity_override="0, 0,-980")
   n4 = Branch(execute=@n3.else, Condition="float > float"(A=Get PawnVelocityZ(), B=50) {is=title K2Node_CallFunction})
   > Set gravity_override(gravity_override="0, 0,-300")
   Set gravity_override(execute=@n4.else, gravity_override="0, 0,-500")
   n5 = Branch(execute=@n1.then_1, Condition=OR Boolean(A=Get bIsSkyDiving(), B=Get bIsSkydiveFloating()))
   > Set frontHairSkydivingAlpha(frontHairSkydivingAlpha:double=0.4) -> (Output_Get:double)
   Set frontHairSkydivingAlpha(execute=@n5.else, frontHairSkydivingAlpha=1)
   Event BlueprintInitializeAnimation()
   > Set Body_Mesh(Body_Mesh=GetOwnerSkeletalMeshComponent())
   > n6 = Cast To PlayerPawn_Athena(Object=TryGetPawnOwner())
   > n7 = Set Pawn Owner(Pawn Owner:Player Pawn Athena Object Reference=@n6) -> (Output_Get:Player Pawn Athena Object Reference)
   > Set Head_Mesh(Head_Mesh:Skeletal Mesh Component Object Reference=GetSkeletalMeshForPartType(self=@n7, PartType=Head)) -> (Output_Get:Skeletal Mesh Component Object Reference)
   TryGetPawnOwner()
  ]}
]
variables [
 Body_Mesh {type=SkeletalMeshComponent}
 frontHairSkydivingAlpha {type=number}
 gravity_override {type=Vector}
 Head_Mesh {type=SkeletalMeshComponent}
 Pawn Owner {type=PlayerPawn_Athena_C}
]
